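# Master material for the female Buffet character body, adding sequins and fake-refractive crystal to the shared character material functions.
# PreFX output gets FX-mask-driven sequins and a metals LUT, then blends into a crystal layer by a selectable mask channel.
# Crystals fake depth with refraction-driven bump offset and LUT sparkle, plus an optional positional weak-point glow with panning caustics.
# PostFX and cinematic functions finish it, with switchable mobile lights.
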Material /Game/Characters/Player/Female/Medium/Bodies/F_MED_Buffet/Materials/M_Buffet_Body_Master
graph {
 properties {
  bUseMaterialAttributes=true
  bUsedWithMeshDeformer=true
  bUsedWithMorphTargets=true
  bUsedWithSkeletalMesh=true
  pixelDepthOffsetMode=PDOM_Legacy
  shadingModel=MSM_FromMaterialExpression
  usedShadingModels=Default Lit | Subsurface Profile
 }
 nodes [
  Constant() {r=1}
  Constant() {r=0.15}
  n1 = MaterialFunctionCall() {materialFunction=&BaseFN_CharacterFunction_PreFX}
  NamedRerouteDeclaration(None=@n1.Fresnel) {name=existingFresnel, nodeColor=<rgba 0.4766 1 0 1>}
  n2 = GetMaterialAttributes(None=@n1.Result) {attributeGetTypes [#1]}
  n3 = Constant()
  n4 = Multiply(A=TextureCoordinate(), B=ScalarParameter() {defaultValue=8, parameterName=Refracted_Texture_Tile})
  n5 = Constant() {r=1}
  n6 = MaterialFunctionCall!(
   Ray Direction=CameraVectorWS()
   Surface Normal=PixelNormalWS()
   Refractive Index Origin
   Refractive Index Target=ScalarParameter() {defaultValue=2, parameterName=ref}
  ) {materialFunction=/Engine/Functions/Engine_MaterialFunctions01/Vectors/Refract}
  n7 = MaterialFunctionCall!(
   HeightTexture=@n3
   TextureCoordinates=@n4
   ReferencePlane=@n5
   HeightRatioInput=ScalarParameter() {defaultValue=0.5, parameterName=rat}
   CameraVector=@n6.Vector
  ) {materialFunction=/Engine/Functions/Engine_MaterialFunctions02/Utility/BumpOffset_Advanced.BumpOffset_advanced}
  n8 = Multiply(
   A=NamedRerouteUsage()
   B=Multiply(
    A=TextureSampleParameter2D(UVs=@n7.UVs) {parameterName=GemInterior, texture=/Game/Gadgets/Assets/VinderTech_Umbrella/Umbrella_Buffet/Materials/T_SequinDiamond_BG}.RGB
    B=ScalarParameter() {defaultValue=10, parameterName=diamondBGBrightness}
   )
  )
  n9 = Max(
   A=StaticSwitchParameter(True=Multiply(A=@n2.Base Color, B=@n8), False=@n8) {parameterName=UseBaseColorForGem}
   B=VectorParameter() {parameterName=DiamondFloorColor}.RGB
  )
  NamedRerouteDeclaration(None=@n9) {name=DiamondBaseColor, nodeColor=<rgba 1 0.0703 0 1>}
  n10 = VectorParameter() {parameterName=SequinColorInner, defaultValue=<rgba 0.6663 0.575 1 1>}
  n11 = TextureSampleParameter2D() {
   parameterName=Fxmap
   samplerType=SAMPLERTYPE_Masks
   texture=/Game/Characters/Player/Female/Medium/Bodies/F_MED_Buffet/Textures/T_F_MED_Buffet_Body_FX
  }
  n12 = VectorParameter() {parameterName=SequinColorOuter, defaultValue=<rgba 0.9546 0.8542 1 1>}
  n13 = GetMaterialAttributes(None=MaterialFunctionCall!(
    SequinColorInner=Multiply(A=@n2.Base Color, B=@n10.RGB)
    Fresnel=NamedRerouteUsage()
    inMat=@n2.MaterialAttributes
    Mask=@n11.G
    SequinColorOuter=Multiply(A=@n12.RGB, B=@n2.Base Color)
   ) {materialFunction=/Game/Gadgets/Assets/VinderTech_Umbrella/Umbrella_Buffet/Materials/MF_BuffetSequins}.Result) {attributeGetTypes [#2]}
  n14 = ChannelMaskParameter(None=@n11.RGB) {maskChannel=Blue, parameterName=CrystalMaskChannel, defaultValue=<rgba 0 0 1 0>}
  n15 = Subtract(A=Subtract(A=@n13.Metallic, B=@n11.G), B=@n14.R)
  n16 = GetMaterialAttributes(None=MaterialFunctionCall(Mask=Saturate(None=@n15), Materials=@n13.MaterialAttributes) {materialFunction=&MF_MetalsLUT}.Result) {attributeGetTypes [
    #1
    #2
    #3
    #4
   ]}
  n17 = TextureSampleParameter2D(UVs=MaterialFunctionCall(
    HeightTexture=@n3
    TextureCoordinates=@n4
    ReferencePlane=@n5
    HeightRatioInput=ScalarParameter() {defaultValue=0.5, parameterName=rat2}
    CameraVector=@n6.Vector
   ) {materialFunction=&BumpOffset_advanced}.UVs) {parameterName=GemInterior, texture=&T_SequinDiamond_BG}
  n18 = OneMinus(None=NamedRerouteUsage())
  n19 = TextureSampleParameter2D(UVs=@n18) {parameterName=GemLUT, texture=/Game/Gadgets/Assets/VinderTech_Umbrella/Umbrella_Buffet/Materials/T_SequinDiamond_RainbowSparkle_LUT}
  n20 = Multiply(
   A=NamedRerouteUsage()
   B=Multiply(A=Multiply(A=@n17.A, B=@n19.RGB), B=ScalarParameter() {defaultValue=3, parameterName=ChomaticChipBrightness})
  ) {constB=0}
  n21 = VectorParameter() {parameterName=WeakPointGlowColor, defaultValue=<rgba 0.9226 0.6637 1 0>}
  n22 = NamedRerouteUsage()
  n23 = NamedRerouteDeclaration(None=Power(Base=@n18) {constExponent=4}) {name=highContrastFacingAngle, nodeColor=<rgba 0 0.0859 1 1>}
  n24 = MaterialFunctionCall(Vector 3=Subtract(
    A=MaterialFunctionCall() {materialFunction=&BoundingBoxBased_0-1_UVW}.RGB
    B=VectorParameter() {parameterName=GlowLocalPosition, defaultValue=<rgba 0.5 0.5 0.8972 0>}.RGB
   )) {materialFunction=/Engine/Functions/Engine_MaterialFunctions02/Utility/VectorLength}
  n25 = ScalarParameter() {defaultValue=0.65, parameterName=WeakPointGlowSize}
  n26 = OneMinus(None=Saturate(None=Divide(A=@n24.V3 Length, B=@n25) {constB=0.65}))
  n27 = Multiply(
   A=Max(A=@n17.RGB) {constB=0.01}
   B=Multiply(A=Multiply(A=@n23, B=@n26), B=ScalarParameter() {defaultValue=50, parameterName=WeakPointGlowIntensity})
  )
  n28 = NamedRerouteUsage()
  n29 = Multiply(
   A=Panner(Coordinate=TextureCoordinate()) {speedX=0.04, speedY=0.04}
   B=ScalarParameter() {defaultValue=5, parameterName=WeakPointCausticTiling}
  ) {constB=5}
  n30 = Multiply(
   A=Multiply(
    A=Power(Base=@n28)
    B=TextureSample(UVs=@n29) {samplerSource=SSM_Wrap_WorldGroupSettings, texture=/Game/Effects/Fort_Effects/Textures/Tile/T_ShieldCaustics_Tile_Chromatic}.RGB
   )
   B=ScalarParameter() {defaultValue=2.5, parameterName=WeakPointGlowCausticIntensity}
  )
  n31 = Multiply(A=Multiply(A=Multiply(A=@n21.RGB, B=@n22), B=Add(A=@n27, B=@n30)), B=ScalarParameter() {parameterName=WeakPointGlowProgress}) {constB=25}
  n32 = SetMaterialAttributes!(
   MaterialAttributes=@n16.MaterialAttributes
   Shading Model=ShadingModel()
   Curvature or Custom Data 0
   Custom Data 1
   Base Color=@n9
   Roughness=ScalarParameter() {defaultValue=0.2, parameterName=InnerDiamondRoughness}
   Metallic=ScalarParameter() {defaultValue=1, parameterName=innerdiamondMetallic}
   Emissive Color=StaticSwitchParameter(True=Add(A=@n20, B=@n31), False=@n20) {parameterName=UseWeakPointGlow}
  ) {attributeSetTypes [
    #5
    #6
    #7
    #1
    #4
    #2
    #8
   ]}
  n33 = MaterialFunctionCall(
   AO=@n1.AO
   Distance Blend=@n1.AO
   Fresnel=@n1.AO
   Gmap=@n1.AO
   Pre Skinned Local Position=@n1.AO
   Material=BlendMaterialAttributes(A=@n16.MaterialAttributes, B=@n32, Alpha=@n14.R)
  ) {materialFunction=&BaseFN_CharacterFunction_PostFX}
  n34 = MaterialFunctionCall(In=@n33.Result) {materialFunction=&CINE_MF_CharPostFX}
  StaticSwitchParameter(
   True=MaterialFunctionCall!(In=@n34.Result) {materialFunction=/Game/Weapons/WeaponSkins/Wraps/MaterialFunctions/MF_3LightsMobile}.Result
   False=@n34.Result
  ) {parameterName=UseCinematicMobileLights}
 ]
}
