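# Character body material for the female Buffet outfit, adding crystal shards and a glowing smoke underlayer to the shared character shading.
# Panning smoke drives underlayer color and emissive, with a five-tap blurred normal at Epic quality.
# A clear-coat crystal layer fakes depth through refracted bump offset, adding time-shifted rainbow sparkle and glowing cracks.
# The FX map's blue channel blends in crystals before the PostFX functions.
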Material /Game/Characters/Player/Female/Medium/Bodies/F_MED_Buffet/Materials/M_Buffet_Body_Shards
graph {
 properties {
  bUseMaterialAttributes=true
  bUsedWithMeshDeformer=true
  bUsedWithMorphTargets=true
  bUsedWithSkeletalMesh=true
  pixelDepthOffsetMode=PDOM_Legacy
  shadingModel=MSM_FromMaterialExpression
  usedShadingModels=Default Lit | Clear Coat | Subsurface Profile
 }
 nodes [
  n1 = TextureCoordinate()
  n2 = ScalarParameter() {defaultValue=8, parameterName=CrystalInteriorTiling}
  n3 = Time()
  n4 = ScalarParameter() {defaultValue=1, parameterName=RotatorSpeed}
  n5 = MaterialFunctionCall!(
   Ray Direction=CameraVectorWS()
   Surface Normal=PixelNormalWS()
   Refractive Index Origin
   Refractive Index Target=ScalarParameter() {defaultValue=2, parameterName=ref}
  ) {materialFunction=/Engine/Functions/Engine_MaterialFunctions01/Vectors/Refract}
  n6 = MaterialFunctionCall!(
   HeightTexture=Constant()
   TextureCoordinates=Rotator(Coordinate=Multiply(A=@n1, B=@n2), Time=Multiply(A=@n3, B=@n4))
   ReferencePlane=Constant() {r=1}
   HeightRatioInput=ScalarParameter() {defaultValue=0.5, parameterName=rat}
   CameraVector=StaticSwitchParameter(True=Transform(None=@n5.Vector) {transformSourceType=TRANSFORMSOURCE_World, transformType=TRANSFORM_Tangent}, False=@n5.Vector) {parameterName=UseTangentRefraction}
  ) {materialFunction=/Engine/Functions/Engine_MaterialFunctions02/Utility/BumpOffset_Advanced.BumpOffset_advanced}
  n7 = Transform(None=PixelNormalWS()) {transformSourceType=TRANSFORMSOURCE_View}
  n8 = OneMinus(None=NamedRerouteUsage())
  n9 = Add(A=@n8, B=Time() {bOverride_Period=true, period=1})
  Add(A=@n6.UVs, B=Multiply(A=ComponentMask(None=@n7) {g=true, r=true}, B=@n9))
  Add(A=@n9) {constB=-10}
  n10 = Constant3Vector() {constant=<rgba 0 0 1 0>}
  n11 = Time()
  n12 = ComponentMask(None=Transform(None=Multiply(A=@n10, B=@n11)) {transformSourceType=TRANSFORMSOURCE_World, transformType=TRANSFORM_Tangent}) {g=true, r=true}
  Add(A=@n12, B=@n6.UVs)
  n13 = MaterialFunctionCall() {materialFunction=&BaseFN_CharacterFunction_PreFX}
  n14 = GetMaterialAttributes(None=@n13.Result) {attributeGetTypes [#1, #2, #3]}
  n15 = Power(Base=NamedRerouteUsage())
  n16 = MaterialFunctionCall(Float3=PixelNormalWS()) {materialFunction=&BreakOutFloat3Components}
  n17 = TextureSample(UVs=Panner(Coordinate=AppendVector(A=@n15, B=@n16.B)) {speedY=0.1}) {samplerSource=SSM_Wrap_WorldGroupSettings, texture=&T_Inky_Smoke_Tile}
  n18 = LinearInterpolate(
   A=VectorParameter() {parameterName=UnderlayerCoreColor, defaultValue=<rgba 0.0141 0.0491 0.0677 1>}.RGB
   B=VectorParameter() {parameterName=UnderlayerEdgeColor, defaultValue=<rgba 0.1711 0.609 0.9062 1>}.RGB
   Alpha=@n17.R
  )
  n19 = NamedRerouteUsage()
  n20 = VectorParameter() {parameterName=UnderlayerEmis, defaultValue=<rgba 0 11.1141 50 1>}
  n21 = Multiply(A=Multiply(A=@n17.R, B=@n15), B=@n20.RGB)
  n22 = NamedRerouteUsage()
  n23 = TextureSampleParameter2D() {
   group=Texture Input
   parameterName=Normals
   samplerSource=SSM_Wrap_WorldGroupSettings
   samplerType=SAMPLERTYPE_Normal
   texture=/Engine/EngineMaterials/DefaultNormal
  }
  n24 = TextureCoordinate()
  n25 = Constant() {r=1}
  n26 = ScalarParameter() {defaultValue=2048, group=NormalBlur, parameterName=TextureRes}
  n27 = MaterialFunctionCall!(OffsetAmount=Multiply(A=Divide(A=@n25, B=@n26), B=ScalarParameter() {defaultValue=1, group=NormalBlur, parameterName=NormalBlurPixels})) {materialFunction=/Engine/Functions/Engine_MaterialFunctions01/ImageAdjustment/BlurSampleOffsets}
  n28 = TextureSampleParameter2D(UVs=Add(A=@n24, B=@n27."0,1")) {
   group=Texture Input
   parameterName=Normals
   samplerSource=SSM_Wrap_WorldGroupSettings
   samplerType=SAMPLERTYPE_Normal
   texture=&DefaultNormal
  }
  n29 = TextureSampleParameter2D(UVs=Add(A=@n24, B=@n27."1,0")) {
   group=Texture Input
   parameterName=Normals
   samplerSource=SSM_Wrap_WorldGroupSettings
   samplerType=SAMPLERTYPE_Normal
   texture=&DefaultNormal
  }
  n30 = Add(A=@n28.RGB, B=@n29.RGB)
  n31 = TextureSampleParameter2D(UVs=Add(A=@n24, B=@n27."0,-1")) {
   group=Texture Input
   parameterName=Normals
   samplerSource=SSM_Wrap_WorldGroupSettings
   samplerType=SAMPLERTYPE_Normal
   texture=&DefaultNormal
  }
  n32 = TextureSampleParameter2D(UVs=Add(A=@n24, B=@n27."-1,0")) {
   group=Texture Input
   parameterName=Normals
   samplerSource=SSM_Wrap_WorldGroupSettings
   samplerType=SAMPLERTYPE_Normal
   texture=&DefaultNormal
  }
  n33 = Add(A=@n31.RGB, B=@n32.RGB)
  n34 = Divide(A=Add(A=@n23.RGB, B=Add(A=@n30, B=@n33))) {constB=5}
  n35 = GetMaterialAttributes(None=SetMaterialAttributes!(
    MaterialAttributes=@n14.MaterialAttributes
    Base Color=LinearInterpolate(A=@n14.Base Color, B=@n18, Alpha=@n19)
    Emissive Color=LinearInterpolate(A=@n14.Emissive Color, B=@n21, Alpha=@n22)
    Normal=QualitySwitch(Default=@n23.RGB, Epic=@n34)
   ) {attributeSetTypes [#1, #3, #4]}) {attributeGetTypes [#1, #2, #4]}
  n36 = Multiply(A=@n35.Base Color, B=ScalarParameter() {defaultValue=1, parameterName=CrytalTintPremultiply})
  n37 = ScalarParameter() {parameterName=BaseColorCrystalTint}
  n38 = NamedRerouteUsage()
  n39 = TextureSampleParameter2D(UVs=@n6.UVs) {parameterName=GemInterior, texture=/Game/Gadgets/Assets/VinderTech_Umbrella/Umbrella_Buffet/Materials/T_SequinShard_BG}
  n40 = Multiply(A=@n39.RGB, B=ScalarParameter() {defaultValue=10, parameterName=diamondBGBrightness})
  n41 = LinearInterpolate(B=@n36, Alpha=ScalarParameter() {parameterName=EmissiveCrystalTint}) {constA=1}
  n42 = TextureSample(UVs=@n9) {texture=/Game/Gadgets/Assets/VinderTech_Umbrella/Umbrella_Buffet/Materials/T_SequinDiamond_RainbowSparkle_LUT}
  n43 = Multiply(
   A=NamedRerouteUsage()
   B=Multiply(A=Multiply(A=@n39.A, B=@n42.RGB), B=ScalarParameter() {defaultValue=3, parameterName=ChromaticChipBrightness})
  ) {constB=0}
  n44 = TextureSampleParameter2D() {
   parameterName=Fxmap
   samplerType=SAMPLERTYPE_Masks
   texture=/Game/Characters/Player/Female/Medium/Bodies/F_MED_Buffet/Textures/T_F_MED_Buffet_Body_FX
  }
  n45 = Multiply(A=@n42.RGB, B=Power(Base=@n44.A))
  n46 = ScalarParameter() {defaultValue=20, parameterName=Cracks}
  n47 = SetMaterialAttributes!(
   MaterialAttributes=@n35.MaterialAttributes
   Shading Model=ShadingModel() {shadingModel=MSM_ClearCoat}
   Clear Coat or Curvature or Custom Data 0=Constant() {r=1}
   Clear Coat Roughness or Custom Data 1=Constant() {r=0.15}
   Base Color=Multiply(A=LinearInterpolate(B=@n36, Alpha=@n37) {constA=1}, B=Multiply(A=@n38, B=@n40))
   Roughness=ScalarParameter() {defaultValue=0.2, parameterName=InnerDiamondRoughness}
   Metallic=ScalarParameter() {defaultValue=1, parameterName=innerdiamondMetallic}
   Emissive Color=Add(A=Multiply(A=@n41, B=@n43), B=Multiply(A=@n45, B=@n46))
  ) {attributeSetTypes [
    #5
    #6
    #7
    #1
    #8
    #2
    #3
   ]}
  n48 = MaterialFunctionCall(
   AO=@n13.AO
   Distance Blend=@n13.AO
   Fresnel=@n13.AO
   Gmap=@n13.AO
   Pre Skinned Local Position=@n13.AO
   Material=BlendMaterialAttributes(A=@n35.MaterialAttributes, B=@n47, Alpha=@n44.B)
  ) {materialFunction=&BaseFN_CharacterFunction_PostFX}
  MaterialFunctionCall(In=@n48.Result) {materialFunction=&CINE_MF_CharPostFX}
  Multiply(A=@n12, B=@n11)
  Multiply(A=LinearInterpolate(Alpha=@n42.R) {constA=0.5}, B=@n20.RGB)
  NamedRerouteDeclaration(None=@n13.Fresnel) {name=existingFresnel, nodeColor=<rgba 0.4766 1 0 1>}
  NamedRerouteDeclaration(None=Power(Base=@n8, Exp=ScalarParameter() {defaultValue=4, parameterName=HighContrastFacingExp}) {constExponent=4}) {name=highContrastFacingAngle, nodeColor=<rgba 0 0.0859 1 1>}
  NamedRerouteDeclaration(None=@n44.R) {name=underlayerMask, nodeColor=<rgba 0.1484 0 1 1>}
  NamedRerouteUsage()
  NamedRerouteUsage()
  Panner(Coordinate=@n6.UVs, Speed=@n12)
  Power(Base=@n15)
  n49 = ComponentMask(None=SkyAtmosphereLightDirection()) {b=true, g=true, r=true}
  n50 = Multiply(A=CameraVectorWS()) {constB=-1}
  Power(Base=Abs(None=DotProduct(A=@n49, B=@n50))) {constExponent=4}
  StaticSwitchParameter() {parameterName=UseNormalBlur}
  TextureCoordinate()
  VertexNormalWS()
 ]
}
